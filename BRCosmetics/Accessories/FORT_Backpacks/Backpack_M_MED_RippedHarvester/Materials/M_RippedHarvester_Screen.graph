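# Animated emissive screen material for the Ripped Harvester backpack cosmetic.
# It layers a radial background, a camera-shifted map with a scrolling gradient sweep, a decal interface and an arrow rotated by `arrow_direction` from the backpack's part modifier Blueprint.
# Scanlines modulate the summed emissive before shared character PostFX and accessory override functions.
# Sine-gated random noise glitches the UVs, and explicit-derivative samples read the map and decals.
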Material /BRCosmetics/Accessories/FORT_Backpacks/Backpack_M_MED_RippedHarvester/Materials/M_RippedHarvester_Screen
graph {
 properties {
  bUseMaterialAttributes=true
  bUsedWithMeshDeformer=true
  bUsedWithSkeletalMesh=true
  pixelDepthOffsetMode=PDOM_Legacy
  usedShadingModels=Default Lit
 }
 nodes [
  n1 = VectorParameter() {parameterName=BGColor1, defaultValue=<rgba 0.4531 0 0 1>}
  n2 = VectorParameter() {parameterName=BGColor2, defaultValue=<rgba 0.0027 0 0 1>}
  n3 = Multiply(
   A=SphereMask(A=NamedRerouteUsage(), B=Constant() {r=0.5}, Radius=Constant() {r=0.6}, Hardness=Constant())
   B=ScalarParameter() {defaultValue=2, parameterName=BGEmissiveIntensity}
  ) {constB=2}
  n4 = Multiply(
   A=NamedRerouteUsage() {desc=Map Mask}
   B=Multiply(A=NamedRerouteUsage() {desc=Map image}, B=VectorParameter() {parameterName=MapColor, defaultValue=<rgba 0.5711 0.0009 0.0009 1>}.RGB)
  )
  n5 = ScalarParameter() {defaultValue=4, parameterName=MapEmissiveIntensity}
  n6 = Multiply(A=NamedRerouteUsage() {desc=Gradiant used for map scroll effect}) {constB=0.5, desc=Decreasing Map intensity}
  n7 = Frac(None=Multiply(A=Time()) {constB=0.2, desc=Decreasing Speed})
  n8 = Add(A=TextureSample(UVs=Add(A=@n6, B=@n7)) {samplerType=SAMPLERTYPE_LinearColor, texture=/Game/Global/Textures/Default/Masks_and_Gradients/T_VertGradient}.RGB) {constB=0.3, desc=Just Decreasing Map FX Intensity}
  n9 = NamedRerouteUsage()
  n10 = NamedRerouteUsage()
  n11 = Constant() {desc="Scalling down Arrow Texture ", r=0.3}
  n12 = NamedRerouteUsage()
  n13 = TextureSample!(
   UVs=MaterialFunctionCall(
    UVs=MaterialFunctionCall(UVs=@n10, Texture Scale=@n11) {materialFunction=&ScaleUVsByCenter}.UVs
    "Rotation Angle (0-1)"=ScalarParameter() {desc=Receive value from B_Athena_PartModifier_Backpack_RippedHarvester, parameterName=arrow_direction}
   ) {materialFunction=&CustomRotator}.Rotated Values
   Tex
   "DDX(UVs)"=DDX(Value=@n12)
   "DDY(UVs)"=DDY(Value=@n12)
   Apply View MipBias
  ) {
   constCoordinate=1
   mipValueMode=TMVM_Derivative
   samplerSource=SSM_Wrap_WorldGroupSettings
   texture=/BRCosmetics/Accessories/FORT_Backpacks/Backpack_M_MED_RippedHarvester/Textures/T_RippedHarvester_Backpack_Decals
  }
  n14 = Multiply(A=Multiply(A=VectorParameter() {parameterName=ArrowColor, defaultValue=<rgba 0.0666 1 0.8388 1>}.RGB, B=@n13.G), B=NamedRerouteUsage())
  n15 = ScalarParameter() {defaultValue=2, parameterName=ArrowEmissiveIntensity}
  n16 = NamedRerouteUsage()
  n17 = Constant() {desc=Scalling Texture Down, r=0.8}
  n18 = NamedRerouteUsage()
  n19 = TextureSample(
   UVs=Add(
    A=MaterialFunctionCall(UVs=@n16, Texture Scale=@n17) {materialFunction=&ScaleUVsByCenter}.UVs
    B=Constant2Vector() {desc=Positioning Texture, g=0.06, r=0.03}
   )
   "DDX(UVs)"=DDX(Value=@n18)
   "DDY(UVs)"=DDY(Value=@n18)
  ) {
   mipValueMode=TMVM_Derivative
   samplerSource=SSM_Clamp_WorldGroupSettings
   texture=&T_RippedHarvester_Backpack_Decals
  }
  n20 = Multiply(A=@n19.R, B=VectorParameter() {parameterName=InterfaceColor, defaultValue=<rgba 1 0.0666 0.107 1>}.RGB)
  n21 = ScalarParameter() {defaultValue=1, parameterName=InterfaceEmissiveIntensity}
  n22 = Add(
   A=LinearInterpolate(A=LinearInterpolate(A=@n1.RGB, B=@n2.RGB, Alpha=@n3), B=Multiply(A=@n4, B=@n5) {constB=3}, Alpha=Multiply(A=@n8, B=@n9)) {constAlpha=1, constB=0}
   B=Add(A=Multiply(A=@n14, B=@n15) {constB=2}, B=Multiply(A=@n20, B=@n21))
  )
  n23 = OneMinus(None=ScalarParameter() {defaultValue=0.7, parameterName=Scanline Intensity})
  n24 = Multiply(A=NamedRerouteUsage(), B=ScalarParameter() {defaultValue=35, parameterName=ScanlineTile})
  n25 = SetMaterialAttributes!(
   MaterialAttributes
   Base Color
   Specular=Constant() {r=0.8}
   Metallic
   Roughness=Constant()
   Emissive Color=LinearInterpolate(
    A=Multiply(A=@n22, B=@n23) {constB=0.8}
    B=@n22
    Alpha=TextureSample!(UVs=@n24, Tex, Apply View MipBias) {samplerType=SAMPLERTYPE_LinearColor, texture=&T_VertGradient}.R
   )
   Normal
  ) {attributeSetTypes [
    #1
    #2
    #3
    #4
    #5
    #6
   ]}
  n26 = MaterialFunctionCall(Material=@n25) {materialFunction=&BaseFN_CharacterFunction_PostFX}
  MaterialFunctionCall(In=@n26.Result) {materialFunction=&MF_TechArtOverrides_accessories}
  n27 = Constant2Vector() {desc=Positioning Map, g=-0.2, r=-0.2}
  n28 = MaterialFunctionCall() {
   desc=Adding player view postion to the map final position
   materialFunction=&CameraDirectionVector
  }
  n29 = Saturate(None=ComponentMask(None=@n28.Result) {g=true, r=true})
  n30 = NamedRerouteUsage()
  n31 = Multiply(A=Constant2Vector() {g=1, r=1.5}, B=ScalarParameter() {defaultValue=0.7, parameterName=Map Scale}) {constB=0.7}
  n32 = NamedRerouteUsage()
  n33 = TextureSample!(
   UVs=Add(
    A=Multiply(A=@n27, B=@n29)
    B=MaterialFunctionCall(UVs=@n30, Texture Scale=@n31) {materialFunction=&ScaleUVsByCenter}.UVs
   )
   Tex
   "DDX(UVs)"=DDX(Value=@n32)
   "DDY(UVs)"=DDY(Value=@n32)
   Apply View MipBias
  ) {
   mipValueMode=TMVM_Derivative
   samplerSource=SSM_Clamp_WorldGroupSettings
   texture=/BRCosmetics/Accessories/FORT_Backpacks/Backpack_M_MED_RippedHarvester/Textures/T_RippedHarvester_Backpack_Map
  }
  NamedRerouteDeclaration(None=@n33.R) {name=MapR, nodeColor=<rgba 0 1 0 1>}
  n34 = MaterialFunctionCall!(Time, Divisor=Constant() {r=0.333}, Mod=Constant() {r=1000}) {materialFunction=/Game/Effects/Fort_Effects/Materials/_Functions/MF_ModuloTime}
  n35 = ComponentMask(None=MaterialFunctionCall!(
    PrimaryTextureCoordinates=TextureCoordinate() {coordinateIndex=1, vTiling=50}
    2ndOrderTextureCoordinates
    RandomSeed=Round(None=@n34.Unclamped)
    NoiseDensityBias
    Bypass Noise Influence Crosstalk
   ) {materialFunction=/Game/Effects/Fort_Effects/Materials/_Functions/MF_RandomUVSample}.NoiseWithRange) {g=true}
  n36 = Constant()
  n37 = Time()
  n38 = Power(Base=Max(A=Sine(None=@n37) {period=3.6}, B=Sine(None=@n37) {period=2.44})) {constExponent=20}
  n39 = Add(A=Multiply(A=AppendVector(A=@n35, B=@n36), B=Round(None=@n38))) {constB=-0.5}
  n40 = ScalarParameter() {defaultValue=0.06, parameterName=DistortionScale}
  n41 = Add(A=TextureCoordinate() {coordinateIndex=1}, B=Saturate(None=Multiply(A=@n39, B=@n40)))
  n42 = NamedRerouteDeclaration(None=Rotator(Coordinate=NamedRerouteDeclaration(None=@n41) {name=UVs, nodeColor=<rgba 0 0.8125 1 1>}, Time=Constant() {r=0.4})) {name=RotatedUVs, nodeColor=<rgba 0 0.0156 1 1>}
  n43 = CurveAtlasRowParameter(CurveTime=NamedRerouteUsage()) {
   atlas=/BRCosmetics/Accessories/FORT_Backpacks/Backpack_M_MED_RippedHarvester/Textures/CA_RippedHarvester
   curve=/BRCosmetics/Accessories/FORT_Backpacks/Backpack_M_MED_RippedHarvester/Textures/C_RippedHarvester_Numbers
   defaultValue=3
   parameterName=Numbers
   sortPriority=0
  }
  n44 = AppendVector(
   A=NamedRerouteDeclaration(None=ScalarParameter() {defaultValue=4.0696, parameterName=U_Scale}) {name=U_Scale, nodeColor=<rgba 1 0.4688 0 1>}
   B=NamedRerouteDeclaration(None=ScalarParameter() {defaultValue=0.22, parameterName=V_Scale}) {name=V_Scale, nodeColor=<rgba 0 1 0 1>}
  )
  NamedRerouteDeclaration(None=Multiply(
    A=MaterialFunctionCall(UVs=@n42, Texture Scale=@n43.R) {materialFunction=&ScaleUVsByCenter}.UVs
    B=NamedRerouteDeclaration(None=@n44) {name=UVScale, nodeColor=<rgba 0 0.2266 1 1>}
   )) {name=Rotated+ScaledUVs, nodeColor=<rgba 0 0.8828 1 1>}
  NamedRerouteDeclaration(None=@n33.G) {name=MapG, nodeColor=<rgba 0 1 0 1>}
  NamedRerouteDeclaration(None=@n33.B) {name=MapB, nodeColor=<rgba 0 1 0 1>}
  n45 = Add(A=TextureCoordinate() {coordinateIndex=1}) {constB=-0.5}
  n46 = Round(None=Abs(None=Multiply(A=@n45) {constB=3.5}))
  n47 = ComponentMask(None=@n46) {r=true}
  n48 = ComponentMask(None=@n46) {g=true}
  NamedRerouteDeclaration(None=OneMinus(None=Max(A=@n47, B=@n48))) {name=ArrowMaskClamp, nodeColor=<rgba 0 0.2031 1 1>}
  n49 = ScalarParameter() {parameterName=Timestamp}
  n50 = ScalarParameter() {defaultValue=3, parameterName=TransitionLength}
  NamedRerouteDeclaration(None=VertexInterpolator(VS=MaterialFunctionCall(TimestampA=@n49, TransitionLength_In=@n50) {materialFunction=&MF_Timestamp_InOut_01}.Linear).PS) {name=TimeStamp, nodeColor=<rgba 0.125 1 0 1>}
 ]
}
